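# Niagara system for the IconColors vehicle boost cosmetic, a ribbon trail attached to skeletal mesh sockets with painted color variants.
# One emitter spawns two ribbons, with color, contrast and desaturation following the variant and rate, lifetime and velocity following game mode.
# Particles inherit fading source movement, while curves and spawn-group masks set widths for three ribbon renderers.
# A custom position adds an animated sine/cosine Z wave along the ribbon.
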NiagaraSystem /VehicleCosmetics/Boosts/Boost_IconColors/FX/NS_Boost_IconColors
dependencies {
 usedDataInterfaces [
  &NiagaraDataInterfaceArrayFloat
  &NiagaraDataInterfaceCurve
  &NiagaraDataInterfaceSkeletalMesh
  &NiagaraDataInterfaceArrayFloat3
 ]
 usedDynamicInputs [&SelectVectorFromArray]
 usedModules [
  &self:SystemSpawnScript.SystemScriptSource.SystemScriptGraph.NiagaraNodeAssignment_0.SetVariables_#1
  &self:SystemSpawnScript.SystemScriptSource.SystemScriptGraph.NiagaraNodeAssignment_3.SetVariables_#2
  &self:SystemSpawnScript.SystemScriptSource.SystemScriptGraph.NiagaraNodeAssignment_2.SetVariables_#3
  &SystemState
 ]
}
emitters {NE_VC_Boost_Ribbon_Parent002 {
  emitterSpawnScript [/SetVariables(
    Emitter.ContrastTx1:NiagaraFloat=&SelectFloatFromArray(
     Direct Array Index=$System.ColorVariantID
     Float Selection Array=DataInterface() {propertyValues {FloatData [0.5,0,1.25,1,0.75,0.75,1.25,1,1,1,1,1.25,1.25,1]}}
    )
    Emitter.Desaturation:NiagaraFloat=&SelectFloatFromArray(Direct Array Index=$System.ColorVariantID, Float Selection Array=DataInterface() {propertyValues {FloatData [1,0,0,0,0,0,0,0,0,0,0,0,0,0]}})
   ) {isSetParameters=true}]
  emitterUpdateScript [
   &EmitterState()
   &VC_FilteredSocketCount()
   &VC_RateQualityHandler(Default Rate:NiagaraFloat=&MergeHelper_FloatSwitch(0=30, 1=30, 2=30, 3=60, Index=$System.GameModeID, Num=::NewEnumerator2))
   &VC_SpawnRibbon_v1(Number of Ribbons=2, SpawnRate=$Emitter.Spawn Rate)
  ]
  particleSpawnScript [
   &VC_PaintedVariantsOverride(
    Black=<rgba 0.2 0.2 0.2 15>
    BurntSienna=<rgba 0.7 0.3631 0.3325 4>
    Cobalt=<rgba 0.024 0.218 0.8 4>
    Crimson=<rgba 1 0.1 0.1244 4>
    DefaultColor=<rgba 1 1 1 10>
    ForestGreen=<rgba 0 0.6 0.0303 7>
    Grey=<rgba 0.52 0.52 0.52 4>
    Lime=<rgba 0.6375 0.85 0 4>
    Orange=<rgba 1.4 0.35 0 4>
    Pink=<rgba 1.3 0.325 1.1375 4>
    Purple=<rgba 0.4934 0.28 0.8 8>
    Saffron=<rgba 1.4 1.3603 0.21 8>
    SkyBlue=<rgba 0 0.8167 1.4 4>
    TitaniumWhite=<rgba 1.1 1.1 1.1 4>
   )
   &VC_InitializeRibbon_v1(
    Color=$StackContext.Color
    Enable Skeletal Position=true
    Lifetime=&MergeHelper_FloatSwitch(0=1, 1=1, 2=1, 3=0.6, Index=$System.GameModeID, Num=::NewEnumerator2)
    Ribbon Width=50
   )
   &AddVelocity(Velocity=&MergeHelper_VectorSwitch(0=<-1500 0 0>, 1=<-1500 0 0>, 2=<-1500 0 0>, 3=<-2500 0 0>, Index=$System.GameModeID, Num=::NewEnumerator2)) {module=AddVelocity001}
   /VehicleCosmetics/FX/VC_FXParents/Modules/DLR_ProductPreviews(GameModeID:NiagaraInt32=$System.GameModeID, bPreviewVehicleSpeed:NiagaraBool=$System.bPreviewVehicleSpeed)
  ]
  particleUpdateScript [
   &ParticleState()
   &InheritSourceMovement(
    Applied Position Scale=&Multiply_VectorByFloat(
     Vector=<1 1 1>
     Float=&FloatFromCurve(FloatCurve=DataInterface() {propertyValues {Curve {
         defaultValue=0
         keys [
          {arriveTangent=-3.5829, interpMode=RCIM_Cubic, leaveTangent=-3.5829, tangentMode=RCTM_User, time=0, value=1}
          {arriveTangent=0.0899, interpMode=RCIM_Cubic, leaveTangent=0.0899, tangentMode=RCTM_User, time=1, value=0}
         ]
        }}})
    )
    ApplyForce=false
    ApplyPosition=true
    ApplyPositionFalloff=false
   )
   /SetVariables(Particles.RibbonMirror:NiagaraFloat=&MaskFloat(
     Mask Source=0
     Mask Parameter=?
     Mask=0
     Masked Float=&FloatFromCurve(
      FloatCurve=DataInterface() {propertyValues {Curve {
         defaultValue=1
         keys [
          {arriveTangent=4.0303, interpMode=RCIM_Cubic, leaveTangent=4.0303, tangentMode=RCTM_User, time=0, value=0.25}
          {arriveTangent=2.9166, interpMode=RCIM_Cubic, leaveTangent=2.9166, tangentMode=RCTM_User, time=0.2, value=1}
          {tangentMode=RCTM_User, time=1, value=0.25}
         ]
        }}}
      Scale Curve=50
     )
     Passthrough Float=&FloatFromCurve(
      FloatCurve=DataInterface() {propertyValues {Curve {
         defaultValue=1
         keys [
          {arriveTangent=4.0303, interpMode=RCIM_Cubic, leaveTangent=4.0303, tangentMode=RCTM_User, time=0, value=0.25}
          {arriveTangent=2.9166, interpMode=RCIM_Cubic, leaveTangent=2.9166, tangentMode=RCTM_User, time=0.2, value=1}
          {tangentMode=RCTM_User, time=1, value=0.25}
         ]
        }}}
      Scale Curve=-50
     )
    )) {enabled=false, isSetParameters=true}
   &SolveForcesAndVelocity(Clamp Velocity=false, Limit Acceleration=false)
   &ScaleRibbonWidth(Ribbon Width Scale=&Multiply_Float(
     A=$Particles.R1OrangeWidth
     B=&FloatFromCurve(
      FloatCurve=DataInterface() {propertyValues {Curve {
         defaultValue=1
         keys [
          {arriveTangent=4.0303, interpMode=RCIM_Cubic, leaveTangent=4.0303, tangentMode=RCTM_User, time=0, value=0}
          {arriveTangent=2.9166, interpMode=RCIM_Cubic, leaveTangent=2.9166, tangentMode=RCTM_User, time=0.2, value=1}
          {tangentMode=RCTM_User, time=1, value=0.25}
         ]
        }}}
      Scale Curve=0.01
     )
    ))
   &DynamicMaterialParameters(
    Index 0 Param 1=&FloatFromCurve(FloatCurve=DataInterface() {propertyValues {Curve {keys [<at 0.6 0>,<at 0.9 0.05>]}}}, Scale Curve=10)
    Module.Index0Float0Write=true
    Module.Index0Float1Write=true
    Module.Index0Float2Write=true
    Module.Index0Float3Write=true
   )
   &ScaleColor(
    Scale Alpha=&FloatFromCurve(FloatCurve=DataInterface() {propertyValues {Curve {
        defaultValue=1
        keys [
         {arriveTangent=-0.0338, interpMode=RCIM_Cubic, leaveTangent=-0.0338, tangentMode=RCTM_User, time=0, value=0}
         {interpMode=RCIM_Cubic, tangentMode=RCTM_User, time=0.5, value=1}
         {tangentMode=RCTM_User, time=1, value=0}
        ]
       }}})
    ScaleA=true
    ScaleRGB=true
   ) {enabled=false}
   &VC_UpdateRibbon_v1(Enable Skeletal Position=true)
   /SetVariables(Particles.CustomPosition_R1:NiagaraPosition=&AddVectorToPosition(
     Position=$Particles.Position
     Vector=&MakeVector(
      X=0
      Y=0
      Z=&MaskFloat(
       Mask Source=0
       Mask Parameter=?
       Mask=0
       Masked Float=&Cosine(
        Normalized Angle=&Add_Float(
         A=&Multiply_Float(A=$Emitter.LoopedAge, B=-1)
         B=&Multiply_Float(A=$Particles.NormalizedAge, B=1)
        )
        Period=1
        Scale=&FloatFromCurve(
         FloatCurve=DataInterface() {propertyValues {Curve {
            defaultValue=0.3
            keys [{interpMode=RCIM_Cubic, tangentMode=RCTM_User, time=0, value=0}, {interpMode=RCIM_Cubic, tangentMode=RCTM_User, time=0.2, value=1}]
           }}}
         Scale Curve=15
        )
        Bias=0
       )
       Passthrough Float=&Sine(
        Normalized Angle=&Add_Float(
         A=&Multiply_Float(A=$Emitter.LoopedAge, B=-1)
         B=&Multiply_Float(A=$Particles.NormalizedAge, B=1)
        )
        Period=1
        Scale=&FloatFromCurve(
         FloatCurve=DataInterface() {propertyValues {Curve {
            defaultValue=0.3
            keys [{interpMode=RCIM_Cubic, tangentMode=RCTM_User, time=0, value=0}, {interpMode=RCIM_Cubic, tangentMode=RCTM_User, time=0.2, value=1}]
           }}}
         Scale Curve=15
        )
        Bias=0
       )
      )
     )
    )) {isSetParameters=true}
   /SetVariables(
    Particles.R1BlueWidth:NiagaraFloat=&Multiply_Float(
     A=&SelectFloatFromArray(Direct Array Index=&SpawnGroup(), Float Selection Array=DataInterface() {propertyValues {FloatData [0,-50]}})
     B=&FloatFromCurve(
      FloatCurve=DataInterface() {propertyValues {Curve {
         defaultValue=1
         keys [
          {arriveTangent=6.1924, interpMode=RCIM_Cubic, leaveTangent=6.1924, tangentMode=RCTM_User, time=0, value=0.3}
          {arriveTangent=0.2432, interpMode=RCIM_Cubic, leaveTangent=0.2432, tangentMode=RCTM_User, time=0.3, value=1.5}
          {tangentMode=RCTM_User, time=1, value=0.25}
         ]
        }}}
      Scale Curve=1.5
     )
    )
    Particles.R2BlueWidth:NiagaraFloat=&Multiply_Float(
     A=&SelectFloatFromArray(Direct Array Index=&SpawnGroup(), Float Selection Array=DataInterface() {propertyValues {FloatData [50,0]}})
     B=&FloatFromCurve(
      FloatCurve=DataInterface() {propertyValues {Curve {
         defaultValue=1
         keys [
          {arriveTangent=6.1924, interpMode=RCIM_Cubic, leaveTangent=6.1924, tangentMode=RCTM_User, time=0, value=0.3}
          {arriveTangent=0.2432, interpMode=RCIM_Cubic, leaveTangent=0.2432, tangentMode=RCTM_User, time=0.3, value=1.5}
          {tangentMode=RCTM_User, time=1, value=0.25}
         ]
        }}}
      Scale Curve=1.5
     )
    )
   ) {isSetParameters=true}
  ]
  rendererClasses [&NiagaraRibbonRendererProperties]
  renderers [
   {rendererClass=&NiagaraRibbonRendererProperties}
   {rendererClass=&NiagaraRibbonRendererProperties, rendererIndex=1}
   {rendererClass=&NiagaraRibbonRendererProperties, rendererIndex=2}
  ]
 }}
summary {
 systemName=NS_Boost_IconColors
 emitters [{emitterName=NE_VC_Boost_Ribbon_Parent002, rendererClasses [&NiagaraRibbonRendererProperties]}]
 userVariables [
  User.bIsFrontendPreview {type=NiagaraBool}
  User.bIsPainted {type=NiagaraBool}
  User.blsIsCoroColor {type=NiagaraBool}
  User.bPreviewVehicleSpeed {type=NiagaraBool, defaultValue=true}
  User.ColorVariantID {type=NiagaraInt32}
  User.GameModeID {type=NiagaraInt32, defaultValue=3}
  User.PaintedColor {type=LinearColor, description="", defaultValue {struct=LinearColor, value=<rgba 1 1 1 1>}}
  User.PaintedScalar {type=NiagaraFloat, defaultValue=1, description=""}
  User.PortCount {type=NiagaraInt32, defaultValue=1}
 ]
}
